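; Parent Blueprint for extendable paper lantern light props, built mostly in the construction script atop an inherited adjustable light.
; The fixture gets the lantern material, replaced by a tinted dynamic instance when building texture data exists.
; The original static mesh is hidden and duplicated as a workaround for the building actor overriding material assignments.
; A knot mesh spawns at the fixture's RTS_Actor socket with a world-position-offset rope-swing material.
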(event EventBeginPlay
  (|CallParentFunction))

(event Collision|EventActorBeginOverlap
  (params (OtherActor :type Actor%20Object%20Reference))
  (|CallParentFunction OtherActor) ; This node is disabled and will not be called.\nDrag off pins to build functionality.
)

(event EventTick
  (params (DeltaSeconds :type Float%20%28single-precision%29))
  (|CallParentFunction DeltaSeconds) ; This node is disabled and will not be called.\nDrag off pins to build functionality.
)

(fn ApplyCorrectMIDFromOverride
  (bind _lightfixture (Variables|AdjustableLight|GetLightFixture))
  (bind _outtexturedataarray (Building|GetBuildingTextureData))
  (if (> (Utilities|Array|Length _outtexturedataarray) 0)
    (Rendering|Material|CreateDynamicMaterialInstance _lightfixture 0 (Variables|/Script/FortniteGame.BuildingTextureData|GetOverrideMaterial (Utilities|Array|Get(acopy) _outtexturedataarray)) "None" :!class /Script/Engine.PrimitiveComponent)
    (Default|SetColoronSecondaryMesh (Math|Conversions|ToLinearColor(Color) (Variables|/Script/FortniteGame.BuildingTextureData|GetTintColor (Utilities|Array|Get(acopy) _outtexturedataarray))) _lightfixture)))

(fn ConstructionScript
  (|CallParentFunction)
  (Rendering|Material|SetMaterial (Variables|AdjustableLight|GetLightFixture) 0 "/Hermes/Props/LightSource/LightSource_PaperLanterns_A/Materials/MI_LightSource_PaperLanterns_A.MI_LightSource_PaperLanterns_A")
  (CallFunction|ApplyCorrectMIDFromOverride)
  (Rendering|SetVisibility (Variables|Debug|GetStaticMeshComponent) false false)
  ; Ugly workaround for buildingactor forcibly overriding attempts at assigning a unique MID not driven by TextureData asset
  (bind _returnvalue (AddComponent|NODE_AddStaticMeshComponent-0))
  (Components|StaticMesh|SetStaticMesh _returnvalue (Variables|/Script/Engine.StaticMeshComponent|GetStaticMesh (Variables|Debug|GetStaticMeshComponent)))
  ; Moving knot mesh to tail end of construction instead of attaching via component hierarchy
  (bind _returnvalue_2 (AddComponent|NODE_AddStaticMeshComponent-1 :RelativeTransform (Math|Transform|MakeTransform (.location (Transformation|GetSocketTransform (Variables|AdjustableLight|GetLightFixture) "None" "RTS_Actor")) "0.000000,-90.000000,0.000000" (.scale (Transformation|GetSocketTransform (Variables|AdjustableLight|GetLightFixture) "None" "RTS_Actor")))))
  (Components|StaticMesh|SetStaticMesh _returnvalue_2 "/Hermes/ArtTools/Tech_Art/KiteAutoWireTool/CozySoup_KiteAnchor_A/Meshes/SM_CozySoup_KiteAnchor_Knot_A.SM_CozySoup_KiteAnchor_Knot_A")
  ; For dem swings
  (Rendering|Material|SetMaterial _returnvalue_2 0 "/Hermes/Props/LightSource/LightSource_PaperLanterns_A/Materials/MI_LightSource_RopeSwing.MI_LightSource_RopeSwing")
  (Rendering|LOD|SetEvaluateWorldPositionOffset _returnvalue_2 true :!class /Script/Engine.StaticMeshComponent)
  (Rendering|LOD|SetWorldPositionOffsetDisableDistance _returnvalue_2 5000 :!class /Script/Engine.StaticMeshComponent)
  (Rendering|Material|SetDefaultCustomPrimitiveDataFloat (Variables|AdjustableLight|GetDownRod) 3 1.0))
